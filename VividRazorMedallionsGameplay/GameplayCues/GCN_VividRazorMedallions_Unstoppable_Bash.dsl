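; Gameplay cue burst for the VividRazorMedallions Unstoppable Bash, positioning its impact VFX for vehicle or non-vehicle targets.
; OnBurst calls the parent, prints a debug string, then checks the target's tags for a vehicle balloon-release tag, apparently used as a vehicle marker.
; It sets IsVehicle on the first particle component and moves it ahead of and above the target.
; Vehicles use a configurable forward offset, extended for the Battle Bus.
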(fn OnBurst(EitherVFX-preferthisversion)
  (params (MyTarget :type Actor%20Object%20Reference) (Parameters :type Gameplay%20Cue%20Parameters%20Structure%20%28by%20ref%29) (ParticleComponents :type Array%20of%20FXSystem%20Component%20Object%20References) (AudioComponents :type Array%20of%20Audio%20Component%20Object%20References) (BurstCameraShakeInstance :type Camera%20Shake%20Base%20Object%20Reference) (BurstDecalInstance :type Decal%20Actor%20Object%20Reference))
  (bind _output (Utilities|Array|Get(acopy) ParticleComponents))
  (|CallParentFunction MyTarget Parameters ParticleComponents AudioComponents BurstCameraShakeInstance BurstDecalInstance)
  (Development|PrintString 0 "SprintBash" true false "(R=1.000000,G=0.816490,B=0.000000,A=1.000000)" 5.0)
  (bind _returnvalue (GameplayTags|GetOwnedGameplayTags MyTarget :!class /Script/GameplayTags.GameplayTagAssetInterface))
  (if (Utilities|Array|ContainsItem (GameplayTags|BreakGameplayTagContainer _returnvalue) "(TagName=\"Athena.Vehicle.OnEntry.ReleaseBalloons\")")
    (bind _returnvalue_2 (GameplayTags|HasMatchingGameplayTag(Message) MyTarget (Variables|Default|GetBattle%20Bus%20Tag)))
    (Effects|Components|ParticleSystem|SetBoolParameter _output "IsVehicle" true)
    (Transformation|SetWorldLocation _output (Utilities|Operators|Add (Transformation|GetWorldLocation (Variables|/Script/Engine.Actor|GetRootComponent MyTarget)) (* (Transformation|GetForwardVector (Variables|/Script/Engine.Actor|GetRootComponent MyTarget)) (select _returnvalue_2 (+ (Variables|Default|GetExplosionForwardOffset) 3.8e+02) (Variables|Default|GetExplosionForwardOffset))) "0.000000,0.000000,120.000000") false false)
    (else
      (Effects|Components|ParticleSystem|SetBoolParameter _output "IsVehicle" false)
      (Transformation|SetWorldLocation _output (Utilities|Operators|Add (Transformation|GetWorldLocation (Variables|/Script/Engine.Actor|GetRootComponent MyTarget)) (* (Transformation|GetForwardVector (Variables|/Script/Engine.Actor|GetRootComponent MyTarget)) 5e+01) "0.000000,0.000000,40.000000") false false)))
  (return))
